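# Material function exposed to the library that unpacks one float from PackTwoNormalizedFloats into two normalized 0-1 floats, good to about three decimals.
# A Custom HLSL node treats the input as two base-4096 digits, using floor and fmod, then rescales them.
# BreakOutFloat2Components feeds the result to outputs NormalizedFloatA (high part) and NormalizedFloatB (low part).
# A second, bit-shift Custom node is left unconnected, noted as lacking Niagara VM support.
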MaterialFunction /Engine/Functions/Engine_MaterialFunctions03/Math/UnpackTwoNormalizedFloats
graph {
 properties {
  bExposeToLibrary=true
  description="Unpacks a float value encoded by the function PackTwoNormalizedFloats into two normalized floats (0-1) onto one single float. Precision is good to 3 decimal places, starts to lose precision at 4."
 }
 nodes [
  MaterialExpressionFunctionInput_0 = FunctionInput() {inputName=PackedFloat, inputType=FunctionInput_Scalar}
  Custom!(inputfloat=@MaterialExpressionFunctionInput_0) {
   code="float num = 65535.0;\r\nfloat2 temp = 0;\r\ntemp.x = float( (asuint(inputfloat) >> 16) && 0x0000FFFF ) / num;\r\ntemp.y = float( asuint(inputfloat) && 0x0000FFFF ) / num;\r\nreturn temp;"
   desc=Missing VM support in Niagara
   outputType=CMOT_Float2
   inputs [{
     inputName=inputfloat
     input {expression=&self:MaterialExpressionFunctionInput_0}
    }]
  }
  n1 = Custom!(packedfloat=@MaterialExpressionFunctionInput_0) {
   code="int precision = 4096;\r\nfloat2 output = 0;\r\n\r\noutput.y = fmod(packedfloat , precision);\r\noutput.x = floor(packedfloat / precision);\r\noutput /= (precision - 1);\r\nreturn output;"
   outputType=CMOT_Float2
   inputs [{
     inputName=packedfloat
     input {expression=&self:MaterialExpressionFunctionInput_0}
    }]
  }
  n2 = MaterialFunctionCall(Float2=@n1) {materialFunction=&BreakOutFloat2Components}
  FunctionOutput(None=@n2.R) {outputName=NormalizedFloatA}
  FunctionOutput(None=@n2.G) {outputName=NormalizedFloatB, sortPriority=1}
 ]
}
