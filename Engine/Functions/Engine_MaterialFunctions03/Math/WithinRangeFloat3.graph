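# Library material function that tests a float3 Value against RangeMin and RangeMax, outputting 0-or-1 masks for any or all components within range.
# Two Step nodes test the inclusive bounds, and their product is split into channels.
# A saturated sum gives AnyComponentsWithinRange, a scalar despite its per-component description, and a product gives AllComponentsWithinRange.
# Both outputs read R twice plus B, so the G channel is never tested.
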MaterialFunction /Engine/Functions/Engine_MaterialFunctions03/Math/WithinRangeFloat3
graph {
 properties {bExposeToLibrary=true, description=Returns 1 if a value is within the provided range limits. 0 otherwise.}
 nodes [
  n1 = FunctionInput() {bUsePreviewValueAsDefault=true, inputName=RangeMin}
  n2 = FunctionInput() {inputName=Value, sortPriority=2}
  n3 = FunctionInput() {bUsePreviewValueAsDefault=true, inputName=RangeMax, sortPriority=1, previewValue=<1 1 1 1>}
  n4 = MaterialFunctionCall(Color=Multiply(A=Step(Y=@n1, X=@n2), B=Step(Y=@n2, X=@n3))) {materialFunction=&SplitComponents}
  n5 = Add(A=@n4.R, B=@n4.R)
  FunctionOutput(None=Saturate(None=Add(A=@n4.B, B=@n5))) {
   description=Returns a vector with each components set to 1 if the component's value is within the provided range limits. 0 otherwise.
   outputName=AnyComponentsWithinRange
  }
  FunctionOutput(None=Multiply(A=Multiply(A=@n4.R, B=@n4.R), B=@n4.B)) {description=Returns 1 if all components of the value vector are within the provided range. 0 otherwise., outputName=AllComponentsWithinRange}
 ]
}
