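# Material function, exposed to the library, that rotates an input vector by a quaternion supplied as a Vector4 (XYZ vector part, W scalar part).
# It sums three terms from dot and cross products, avoiding a matrix or full quaternion products: 2(q·v)q + (w²−q·q)v + 2w(q×v).
# The optional Invert Quaternion static switch negates XYZ, applying the conjugate rotation.
# The function does not normalize the quaternion.
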MaterialFunction /Engine/Functions/Engine_MaterialFunctions03/Math/MultiplyVectorWithQuaternion
graph {
 properties {bExposeToLibrary=true}
 nodes [
  n1 = FunctionInput() {inputName=Quat, inputType=FunctionInput_Vector4, sortPriority=1}
  n2 = ComponentMask(None=@n1) {b=true, g=true, r=true}
  n3 = StaticSwitch(
   True=Multiply(A=@n2) {constB=-1}
   False=@n2
   Value=FunctionInput(Preview=StaticBool()) {bUsePreviewValueAsDefault=true, inputName=Invert Quaternion, inputType=FunctionInput_StaticBool, sortPriority=2}
  )
  n4 = FunctionInput() {inputName=Vector}
  n5 = Multiply(A=@n3, B=Multiply(A=DotProduct(A=@n3, B=@n4)) {constB=2})
  n6 = ComponentMask(None=@n1) {a=true}
  n7 = Multiply(A=Subtract(A=Multiply(A=@n6, B=@n6), B=DotProduct(A=@n3, B=@n3)), B=@n4)
  n8 = CrossProduct(A=@n3, B=@n4)
  n9 = Multiply(A=@n6) {constB=2}
  FunctionOutput(None=Add(A=Add(A=@n5, B=@n7), B=Multiply(A=@n8, B=@n9)))
 ]
}
