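# WrapFloat is a library material function in the Math category that wraps a scalar Value into the range between 0 and Wrap.
# It divides Value by Wrap, floors the result, then subtracts that multiple of Wrap from Value.
# Using floor gives fmod-like behavior that also handles negative inputs, and the Wrap input uses its preview value of 1 as the default.
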MaterialFunction /Engine/Functions/Engine_MaterialFunctions03/Math/WrapFloat
graph {
 properties {
  bExposeToLibrary=true
  description=Similar to fmod with support for negative floats : returns a value between 0 and Wrap
  libraryCategoriesText [Math]
 }
 nodes [
  n1 = FunctionInput() {desc=Value to wrap, inputName=Value, inputType=FunctionInput_Scalar}
  n2 = FunctionInput() {
   bUsePreviewValueAsDefault=true
   desc=Interval within which the input value should be wrapped
   inputName=Wrap
   inputType=FunctionInput_Scalar
   sortPriority=1
   previewValue=<1 1 1 1>
  }
  n3 = Floor(None=Divide(A=@n1, B=@n2))
  FunctionOutput(None=Subtract(A=@n1, B=Multiply(A=@n2, B=@n3))) {description=Wrapped around value}
 ]
}
